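; Gas station trash can Blueprint that can release a swarm of flies onto nearby players when looted.
; The loot event passes through a DoOnce, then a 50% random roll activates the can's flies effect.
; A box overlap around the can gathers player pawns, attaching a flies emitter to each one's head bone and a buzzing sound to its mesh.
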(event EventOnLoot
  (Utilities|FlowControl|DoOnce false
    (:Completed
      (if (< (Math|Random|RandomFloatinRange 0.0 100) 50)
        (Components|Activation|Activate (Variables|Utility_TrashCanLarge_01|GetEffect_Flies) false)
        (bind _outactors (Collision|BoxOverlapActors (Transformation|GetActorLocation) "256.000000,256.000000,256.000000" (Variables|Default|GetObjectTypeArray) "/Script/FortniteGame.FortPlayerPawn" :!class /Script/Engine.KismetSystemLibrary))
        (for _index (range (+ (Utilities|Array|LastIndex _outactors) 1))
          (bind _asfort_player_pawn (Utilities|Casting|CastToFortPlayerPawn (Utilities|Array|Get(acopy) _outactors _index)))
          (Effects|SpawnEmitterAttached "/Game/Effects/Fort_Effects/World/Enviroment/Flies/P_Trash_Flies_around_player.P_Trash_Flies_around_player" (Variables|/Script/Engine.Character|GetMesh _asfort_player_pawn) "b_MF_Head")
          (Audio|SpawnSoundAttached :Sound "/Game/Sounds/Fort_Ambient_Looping/Nature_Exterior/Insect_Flies_Loop_Env5Secs_Cue.Insect_Flies_Loop_Env5Secs_Cue" :AttachToComponent (Variables|/Script/Engine.Character|GetMesh _asfort_player_pawn) :AttachPointName "None" :LocationType "KeepRelativeOffset" :bStopWhenAttachedToDestroyed false :VolumeMultiplier 2 :PitchMultiplier 1.0 :StartTime 0.0 :bAutoDestroy true))))))

(fn ConstructionScript)
